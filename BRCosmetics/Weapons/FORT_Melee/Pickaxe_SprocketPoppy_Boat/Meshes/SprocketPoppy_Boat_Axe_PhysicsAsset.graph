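# Physics asset for the SprocketPoppy Boat axe, a Fortnite cosmetic pickaxe melee weapon mesh.
# It defines a single capsule body on the two-handed handle bone, offset and slightly tilted along the weapon.
# There are no constraints, so the whole weapon acts as one rigid collision shape.
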PhysicsAsset /BRCosmetics/Weapons/FORT_Melee/Pickaxe_SprocketPoppy_Boat/Meshes/SprocketPoppy_Boat_Axe_PhysicsAsset
bodies {Melee_TwoHanded_Handle [{shapeType=Capsule, length=36.8403, radius=17.7209, center=<0.0049 -0.7133 25.3866>, rotation=<rot -9.4722 -90 -0.0262>}]}
constraints []
